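# Material function for the FlorWireframe cosmetic wheel, overriding emissive and normal on MF_WheelParent_Simple to draw a glowing wireframe look.
# Vertex red blends rim and tire maps, whose blue-channel wires are layered at several BumpOffset depths for fake parallax and combined with Fresnel.
# Red and green channels drive paint masks with time-stepped stripes, optionally rotating or panning.
# Panning screen-space scanlines and switches for inversion, paint and secondary color complete the setup.
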MaterialFunction /VehicleCosmetics/Wheels/FlorWireframe/Materials/MF_Wheel_FlorWireframe
graph {nodes [
  Constant()
  Constant()
  Constant()
  n1 = NamedRerouteUsage()
  n2 = NamedRerouteUsage()
  n3 = ConstantBiasScale(None=@n2) {bias=0, scale=1}
  n4 = BumpOffset(Coordinate=NamedRerouteUsage(), Height=Add(B=NamedRerouteUsage()) {constA=0.65}) {heightRatio=-0.05}
  n5 = TextureSampleParameter2D(UVs=@n4) {
   group=1_Input Textures
   parameterName=RimWireframeMap
   samplerType=SAMPLERTYPE_LinearColor
   texture=/VehicleCosmetics/Wheels/FlorWireframe/Textures/T_Rim_FlorWireframe_RGB_FX
  }
  n6 = TextureSampleParameter2D(UVs=@n4) {
   group=1_Input Textures
   parameterName=TireWireframeMap
   samplerType=SAMPLERTYPE_LinearColor
   sortPriority=31
   texture=/VehicleCosmetics/Wheels/FlorWireframe/Textures/T_Tire_FlorWirefarme_RGB_FX
  }
  n7 = VertexColor()
  n8 = LinearInterpolate(A=@n5.B, B=@n6.B, Alpha=@n7.R)
  n9 = OneMinus(None=@n8)
  n10 = BumpOffset(Coordinate=NamedRerouteUsage(), Height=Add(B=NamedRerouteUsage()) {constA=0.9}) {heightRatio=-0.05}
  n11 = TextureSampleParameter2D(UVs=@n10) {
   group=1_Input Textures
   parameterName=RimWireframeMap
   samplerType=SAMPLERTYPE_LinearColor
   texture=&T_Rim_FlorWireframe_RGB_FX
  }
  n12 = TextureSampleParameter2D(UVs=@n10) {
   group=1_Input Textures
   parameterName=TireWireframeMap
   samplerType=SAMPLERTYPE_LinearColor
   sortPriority=31
   texture=&T_Tire_FlorWirefarme_RGB_FX
  }
  n13 = LinearInterpolate(A=@n11.B, B=@n12.B, Alpha=@n7.R)
  n14 = OneMinus(None=@n13)
  n15 = Max(
   A=StaticSwitchParameter(True=Saturate(None=@n9), False=@n8) {group=1_3_SwitchProperties, parameterName=WireInvert}
   B=StaticSwitchParameter(True=Saturate(None=@n14), False=@n13) {group=1_3_SwitchProperties, parameterName=WireInvert}
  )
  n16 = LinearInterpolate(A=Multiply(A=@n15) {constB=0.5}, B=@n15, Alpha=NamedRerouteUsage())
  n17 = NamedRerouteUsage()
  n18 = BumpOffset(Coordinate=NamedRerouteUsage(), Height=Add(B=NamedRerouteUsage()) {constA=0.65}) {heightRatio=-0.05}
  n19 = TextureSampleParameter2D(UVs=@n18) {
   group=1_Input Textures
   parameterName=RimWireframeMap
   samplerType=SAMPLERTYPE_LinearColor
   texture=&T_Rim_FlorWireframe_RGB_FX
  }
  n20 = TextureSampleParameter2D(UVs=@n18) {
   group=1_Input Textures
   parameterName=TireWireframeMap
   samplerType=SAMPLERTYPE_LinearColor
   sortPriority=31
   texture=&T_Tire_FlorWirefarme_RGB_FX
  }
  n21 = VertexColor()
  n22 = LinearInterpolate(A=@n19.B, B=@n20.B, Alpha=@n21.R)
  n23 = StaticSwitchParameter(True=Saturate(None=OneMinus(None=@n22)), False=@n22) {group=1_3_SwitchProperties, parameterName=WireInvert}
  n24 = Multiply(B=@n23) {constA=0.5}
  n25 = Multiply(A=@n23) {constB=0.9}
  n26 = NamedRerouteUsage()
  n27 = NamedRerouteUsage()
  n28 = TextureSampleParameter2D(UVs=@n27) {
   group=1_Input Textures
   parameterName=RimWireframeMap
   samplerType=SAMPLERTYPE_LinearColor
   texture=&T_Rim_FlorWireframe_RGB_FX
  }
  n29 = TextureSampleParameter2D(UVs=@n27) {
   group=1_Input Textures
   parameterName=TireWireframeMap
   samplerType=SAMPLERTYPE_LinearColor
   sortPriority=31
   texture=&T_Tire_FlorWirefarme_RGB_FX
  }
  n30 = LinearInterpolate(A=@n28.B, B=@n29.B, Alpha=@n21.R)
  n31 = StaticSwitchParameter(True=Saturate(None=OneMinus(None=@n30)), False=@n30) {group=1_3_SwitchProperties, parameterName=WireInvert}
  n32 = Saturate(None=Max(A=LinearInterpolate(A=@n24, B=@n25, Alpha=@n26), B=Multiply(A=@n31) {constB=0.7}))
  n33 = Saturate(None=ConstantBiasScale(None=@n32) {bias=0, scale=3})
  n34 = Multiply(A=NamedRerouteUsage(), B=NamedRerouteUsage())
  n35 = Constant() {r=1}
  n36 = Fresnel(Normal=Transform(None=AppendVector(A=@n34, B=@n35))) {baseReflectFraction=0.02, exponent=3}
  n37 = Power(Base=@n36) {constExponent=1}
  n38 = Power(Base=@n36) {constExponent=30}
  n39 = NamedRerouteUsage()
  n40 = OneMinus(None=@n36)
  n41 = Saturate(None=Power(Base=@n40) {constExponent=20})
  n42 = NamedRerouteUsage()
  n43 = Max(A=Saturate(None=LinearInterpolate(A=@n37, B=@n38, Alpha=@n39)), B=ConstantBiasScale(None=Multiply(A=@n41, B=@n42)) {bias=0, scale=0.9})
  n44 = Saturate(None=@n43)
  n45 = LinearInterpolate(
   A=NamedRerouteUsage()
   B=LinearInterpolate(A=LinearInterpolate(A=@n1, B=@n3, Alpha=@n16), B=Multiply(A=@n33, B=@n17) {constA=0.5}, Alpha=@n33)
   Alpha=Saturate(None=Add(A=@n44, B=@n32))
  )
  n46 = ConstantBiasScale(None=@n43) {bias=0.1, scale=1}
  n47 = Multiply(A=Multiply(A=NamedRerouteUsage(), B=Max(A=@n46, B=@n33)), B=NamedRerouteUsage())
  n48 = NamedRerouteUsage()
  n49 = Saturate(None=StaticSwitchParameter(True=LinearInterpolate(A=@n45, B=@n47, Alpha=@n48), False=@n45) {defaultValue=true, group=1_3_SwitchProperties, parameterName=Painted})
  n50 = NamedRerouteUsage()
  FunctionOutput(None=SetMaterialAttributes!(
    MaterialAttributes=MaterialFunctionCall() {materialFunction=&MF_WheelParent_Simple}.MaterialAttributes
    EmissiveColor=Multiply(A=@n49, B=@n50)
    Normal=NamedRerouteUsage()
   ) {attributeSetTypes [#1, #2]})
  n51 = NamedRerouteUsage()
  n52 = NamedRerouteUsage()
  n53 = NamedRerouteUsage()
  NamedRerouteDeclaration(None=StaticSwitchParameter(True=LinearInterpolate(A=@n51, B=@n52, Alpha=@n53), False=@n51) {group=1_3_SwitchProperties, parameterName=SecondaryColor}) {name=Colors, nodeColor=<rgba 0 1 0.7188 1>}
  n54 = Saturate(None=Power(Base=Saturate(None=@n40)) {constExponent=10})
  NamedRerouteDeclaration(None=Multiply(A=@n43, B=LinearInterpolate(A=@n3, B=@n2, Alpha=@n54))) {name=Fresnel, nodeColor=<rgba 0.1484 0 1 1>}
  NamedRerouteDeclaration(None=VectorParameter() {group=1_2_WireFrameColor, parameterName=PaintColor, defaultValue=<rgba 0.0014 0.2674 0.7227 1>}.RGB) {name=PaintColor, nodeColor=<rgba 0.9922 0 1 1>}
  NamedRerouteDeclaration(None=VectorParameter() {group=1_2_WireFrameColor, parameterName=SecondaryColor, defaultValue=<rgba 1 0.3392 0 1>}.RGB) {name=SecondaryColor, nodeColor=<rgba 1 0 0.375 1>}
  NamedRerouteDeclaration(None=ScalarParameter() {defaultValue=375, group=1_1_WireFrameProperties, parameterName=Scanline Tiling Multiplier}) {name=Scanline Tiling Multiplier, nodeColor=<rgba 0 0.2266 1 1>}
  NamedRerouteDeclaration(None=ScalarParameter() {defaultValue=-1.5, group=1_1_WireFrameProperties, parameterName=Effect Normal}) {name=Effect Normal, nodeColor=<rgba 1 0.7969 0 1>}
  n55 = MaterialFunctionCall() {materialFunction=&MF_WheelParent_Simple}
  n56 = GetMaterialAttributes(None=@n55.MaterialAttributes) {attributeGetTypes [#2]}
  n57 = ComponentMask(None=@n56.Normal) {r=true}
  n58 = ComponentMask(None=@n56.Normal) {g=true}
  NamedRerouteDeclaration(None=AppendVector(A=LinearInterpolate(A=@n57, B=@n57, Alpha=@n55.VertexColorR), B=LinearInterpolate(A=@n58, B=@n58, Alpha=@n55.VertexColorR))) {name=Normal, nodeColor=<rgba 0.0078 1 0 1>}
  NamedRerouteDeclaration(None=ScalarParameter() {defaultValue=1, group=1_1_WireFrameProperties, parameterName=Surface Normal}) {name=Surface Normal, nodeColor=<rgba 1 0 0.3281 1>}
  n59 = NamedRerouteUsage()
  n60 = NamedRerouteUsage()
  NamedRerouteDeclaration(None=AppendVector(A=Multiply(A=@n59, B=@n60), B=Constant() {r=1})) {name=SurfaceNormalColor, nodeColor=<rgba 0 0.6484 1 1>}
  n61 = Cosine(None=Multiply(A=Time()) {constB=0.25})
  NamedRerouteDeclaration(None=Multiply(A=Abs(None=@n61)) {constB=0.05}) {name=Timer, nodeColor=<rgba 0.3828 0 1 1>}
  NamedRerouteDeclaration(None=Multiply(A=ScalarParameter() {defaultValue=1, group=1_1_WireFrameProperties, parameterName=WireUVTiling}, B=TextureCoordinate())) {name=UVTilling, nodeColor=<rgba 0.0078 0 1 1>}
  n62 = ComponentMask(None=ScreenPosition().ViewportUV) {g=true, r=true}
  n63 = NamedRerouteUsage()
  n64 = TextureSampleParameter2D(UVs=Panner(Coordinate=Multiply(A=@n62, B=@n63)) {speedY=1}) {
   group=1_Input Textures
   parameterName="ScanlinesMap (R)"
   sortPriority=33
   texture=/VehicleCosmetics/Textures/UniversalTextures/T_LinearGradient_RGB
  }
  NamedRerouteDeclaration(None=@n64.R) {name=Scanlines, nodeColor=<rgba 0.8281 1 0 1>}
  NamedRerouteDeclaration(None=Saturate(None=@n64.R)) {name=Scanlines_clamp, nodeColor=<rgba 1 0 0.6797 1>}
  NamedRerouteDeclaration(None=ScalarParameter() {defaultValue=2, group=1_1_WireFrameProperties, parameterName=HighlightMultiplier}) {name=HighlightMultiplier, nodeColor=<rgba 0 1 0.4141 1>}
  n65 = NamedRerouteUsage()
  n66 = Multiply(B=Power(Base=@n16)) {constA=3}
  NamedRerouteDeclaration(None=LinearInterpolate(A=Multiply(A=@n65) {constA=3, constB=0.02}, B=@n3, Alpha=Saturate(None=@n66))) {name=AccentColor, nodeColor=<rgba 0 0.7656 1 1>}
  NamedRerouteDeclaration(None=ScalarParameter() {defaultValue=1, group=1_1_WireFrameProperties, parameterName=GreenpanSpeed}) {name=GreenpanSpeed, nodeColor=<rgba 1 0.8203 0 1>}
  NamedRerouteDeclaration(None=ScalarParameter() {defaultValue=1, group=1_1_WireFrameProperties, parameterName=Time Multiplier}) {name=TimeMultiplier, nodeColor=<rgba 0.9688 1 0 1>}
  n67 = ConstantBiasScale(None=NamedRerouteUsage()) {bias=0, scale=10}
  n68 = Time()
  n69 = Multiply(A=Floor(None=@n68), B=NamedRerouteUsage())
  n70 = TextureSampleParameter2D(UVs=Rotator(Time=@n69) {centerX=0.2743, centerY=0.6951}) {
   group=1_Input Textures
   parameterName=RimWireframeMap
   samplerType=SAMPLERTYPE_LinearColor
   texture=&T_Rim_FlorWireframe_RGB_FX
  }
  n71 = StaticSwitchParameter(True=Multiply(A=Saturate(None=@n67), B=@n70.R), False=NamedRerouteUsage()) {group=1_3_SwitchProperties, parameterName=Tread Greenpan}
  n72 = NamedRerouteUsage()
  n73 = VertexColor()
  n74 = Multiply(A=@n68, B=NamedRerouteUsage())
  n75 = Multiply(A=Abs(None=Sine(None=@n74) {period=3})) {constB=60}
  n76 = OneMinus(None=Divide(A=Floor(None=@n75)) {constB=30})
  n77 = Clamp(None=Subtract(A=LinearInterpolate(A=@n71, B=@n72, Alpha=@n73.R), B=Saturate(None=@n76))) {maxDefault=0.75}
  n78 = Saturate(None=ConstantBiasScale(None=NamedRerouteUsage()) {bias=0, scale=10})
  n79 = TextureSampleParameter2D(UVs=Panner(Time=@n69) {speedX=0.1}) {
   group=1_Input Textures
   parameterName=TireWireframeMap
   samplerType=SAMPLERTYPE_LinearColor
   sortPriority=31
   texture=&T_Tire_FlorWirefarme_RGB_FX
  }
  n80 = LinearInterpolate(
   A=NamedRerouteUsage()
   B=StaticSwitchParameter(True=Multiply(A=@n78, B=@n79.G), False=NamedRerouteUsage()) {group=1_3_SwitchProperties, parameterName=Tread Greenpan}
   Alpha=@n73.R
  )
  n81 = Multiply(A=Sine(None=Subtract(A=@n74) {constB=0.25}) {period=3}) {constB=60}
  n82 = OneMinus(None=Divide(A=Floor(None=@n81)) {constB=30})
  n83 = Clamp(None=Subtract(A=@n80, B=Saturate(None=@n82))) {maxDefault=0.75}
  n84 = ConstantBiasScale(None=Clamp(None=Max(A=@n77, B=@n83)) {maxDefault=0.75}) {bias=0, scale=5}
  NamedRerouteDeclaration(None=@n84) {name=PaintedStripe, nodeColor=<rgba 0 0.9766 1 1>}
  NamedRerouteDeclaration(None=ConstantBiasScale(None=@n84) {bias=0.2, scale=1}) {name=PaintedStripe_Light, nodeColor=<rgba 0 1 0.6953 1>}
  NamedRerouteDeclaration(None=@n80) {name=PaintedStripe_Tire, nodeColor=<rgba 1 0.375 0 1>}
  n85 = TextureSampleParameter2D() {
   group=1_Input Textures
   parameterName=RimWireframeMap
   samplerType=SAMPLERTYPE_LinearColor
   texture=&T_Rim_FlorWireframe_RGB_FX
   channelNames {a="", b=Wirreframe Decals, g=SeconddaryColor, r=PaintedColor}
  }
  NamedRerouteDeclaration(None=ComponentMask(None=@n85.RGB) {r=true}) {name=R_RimWireframe, nodeColor=<rgba 0 1 1 1>}
  NamedRerouteDeclaration(None=ComponentMask(None=@n85.RGB) {g=true}) {name=G_RimWireframe, nodeColor=<rgba 1 0.6797 0 1>}
  NamedRerouteDeclaration(None=ComponentMask(None=@n85.RGB) {b=true}) {name=B_RimWireframe, nodeColor=<rgba 0.3594 0 1 1>}
  n86 = TextureSampleParameter2D() {
   group=1_Input Textures
   parameterName=TireWireframeMap
   samplerType=SAMPLERTYPE_LinearColor
   sortPriority=31
   texture=&T_Tire_FlorWirefarme_RGB_FX
   channelNames {a="", b=Wirreframe Decals, g=SeconddaryColor, r=PaintedColor}
  }
  NamedRerouteDeclaration(None=ComponentMask(None=@n86.RGB) {r=true}) {name=R_TireWireframe, nodeColor=<rgba 0 1 0.0391 1>}
  NamedRerouteDeclaration(None=ComponentMask(None=@n86.RGB) {g=true}) {name=G_TireWireframe, nodeColor=<rgba 1 0 0.2812 1>}
  NamedRerouteDeclaration(None=ComponentMask(None=@n86.RGB) {b=true}) {name=B_TireWireframe, nodeColor=<rgba 0 0.6016 1 1>}
  NamedRerouteDeclaration(None=ScalarParameter() {defaultValue=1.5, group=1_2_WireFrameColor, parameterName=EmissiveIntensity}) {name=EmissIntensity, nodeColor=<rgba 0.6875 1 0 1>}
 ]}
