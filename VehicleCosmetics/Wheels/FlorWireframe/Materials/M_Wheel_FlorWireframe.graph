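# M_Wheel_FlorWireframe is the material for the FlorWireframe vehicle wheel cosmetic, built on material attributes with Default Lit shading.
# It chains three material functions: MF_Wheel_FlorWireframe supplies the base look, MF_BaseVehicle_Damage layers on damage, and MF_VehicleDeRez_Transition applies the de-rez transition last.
# The surface logic lives entirely in those functions, and usage is limited to skeletal mesh and mesh deformer flags, with automatic usage setting disabled.
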Material /VehicleCosmetics/Wheels/FlorWireframe/Materials/M_Wheel_FlorWireframe
graph {
 properties {
  bAutomaticallySetUsageInEditor=false
  bUseMaterialAttributes=true
  bUsedWithMeshDeformer=true
  bUsedWithSkeletalMesh=true
  usedShadingModels=Default Lit
 }
 nodes [MaterialFunctionCall(In=MaterialFunctionCall(Material=MaterialFunctionCall() {materialFunction=&/MF_Wheel_FlorWireframe}.Result) {materialFunction=&MF_BaseVehicle_Damage}.Result) {materialFunction=&MF_VehicleDeRez_Transition}]
}
